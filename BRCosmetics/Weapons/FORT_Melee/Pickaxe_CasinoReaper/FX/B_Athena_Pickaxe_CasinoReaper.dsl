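; Cosmetic weapon Blueprint for the Casino Reaper pickaxe that adjusts the orientation of its impact effect.
; The play-impact-FX event calls the parent, then checks whether the cosmetic is displayed in a frontend context.
; If so, it sets the Impact FX world rotation to the instigator pawn's rotation combined with a 180-degree yaw.
; BeginPlay and Tick are empty, and the construction script only calls the parent.
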(event EventBeginPlay)

(event EventTick
  (params (DeltaSeconds :type Float%20%28single-precision%29)))

(event FortWeapon|EventOnPlayImpactFX
  (params (HitResult :type Hit%20Result%20Structure%20%28by%20ref%29) (ImpactPhysicalSurface :type EPhysicalSurface%20Enum :default "SurfaceType_Default") (SpawnedPSC :type FXSystem%20Component%20Object%20Reference))
  (|CallParentFunction HitResult ImpactPhysicalSurface SpawnedPSC)
  (if (Reactive|Frontend|IsCosmeticObjectDisplayedinFrontendContext 0)
    (bind _asfort_player_pawn (Utilities|Casting|CastToFortPlayerPawn (Game|GetInstigator)))
    (Transformation|SetWorldRotation (Variables|Default|GetImpact%20FX) (Math|Rotator|CombineRotators (Transformation|GetActorRotation _asfort_player_pawn) "0.000000,180.000000,0.000000") false false)))

(fn ConstructionScript
  (|CallParentFunction))
