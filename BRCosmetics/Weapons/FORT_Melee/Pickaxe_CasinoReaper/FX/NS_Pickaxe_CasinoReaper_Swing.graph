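# Niagara swing effect for the CasinoReaper pickaxe, drawing two short-lived ribbon anim trails along the weapon.
# A socket reader feeds both emitters: Trail makes the wide, faint sweep, while TipTrail makes a thin streak, sized from the second socket, that tapers with age.
# User.VariantIndex picks each trail's base color through a color switch.
# Both fade over particle life, with tint and opacity adjusted for time of day.
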NiagaraSystem /BRCosmetics/Weapons/FORT_Melee/Pickaxe_CasinoReaper/FX/NS_Pickaxe_CasinoReaper_Swing
dependencies {
 usedDataInterfaces [&NiagaraDataInterfaceCurve, &NiagaraDataInterfaceSocketReader]
 usedDynamicInputs []
 usedModules [
  &self:SystemSpawnScript.SystemScriptSource.SystemScriptGraph.NiagaraNodeAssignment_0.SetVariables_#1
  &SystemState
 ]
}
emitters {
 TipTrail {
  emitterUpdateScript [
   &EmitterState(Life Cycle Mode=::NewEnumerator1, Loop Behavior=::NewEnumerator1, Loop Duration=0.27, UseLoopDelay=false)
   &FN_SpawnAnimTrail_v2(AnimTrailWidthMode=::FromSecond)
  ]
  particleSpawnScript [
   /Game/Effects/Niagara/Modules/AnimTrail/FN_InitializeAnimTrail/FN_InitializeAnimTrailBoneIndecies(Second Bone or Socket Index=1)
   /Game/Effects/Niagara/Modules/AnimTrail/FN_InitializeAnimTrail/FN_InitializeAnimTrail_SocketReader/FN_InitializeAnimTrail_SocketReader(
    AnimTrailWidth=0.1
    Color:LinearColor=&MultiplyLinearColor(
     Linear Color=&MergeHelper_ColorSwitch(
      0=<rgba 0.6198 0.7085 1 0.5>
      1=<rgba 0.3718 0.8554 1 0.5>
      2=?
      3=?
      Index=&MakeIntFromFloat(Float=$User.VariantIndex)
      Num=::NewEnumerator0
     )
     Scale Factor=&MakeFloat4FromVectorAndFloat(
      "Vector (RGB)"=<1 1 1>
      "Float (Alpha)"=&FloatFromCurve(FloatCurve=DataInterface() {propertyValues {Curve {keys [<at 0.2 1>,<at 0.25 1>]}}}, CurveIndex=$System.Age)
     )
    )
    Lifetime=0.15
    Socket Reader:NiagaraDataInterfaceSocketReader=$System.Socket Reader
   )
   &AddVelocity(Velocity=<2000 0 0>) {enabled=false}
  ]
  particleUpdateScript [
   &SolveForcesAndVelocity(Clamp Velocity=false, Limit Acceleration=false)
   &ScaleRibbonWidth(Ribbon Width Scale=&FloatFromCurve(FloatCurve=DataInterface() {propertyValues {Curve {keys [
         {arriveTangent=-1.2579, interpMode=RCIM_Cubic, time=0, value=1}
         {arriveTangent=-0.9664, interpMode=RCIM_Cubic, leaveTangent=-0.9664, tangentMode=RCTM_User, time=1, value=0}
        ]}}}))
   &ScaleColor(
    Scale Alpha=&Multiply_Float(
     A=&FloatFromCurve(FloatCurve=DataInterface() {propertyValues {Curve {keys [{interpMode=RCIM_Cubic, leaveTangent=-2.3646, tangentMode=RCTM_Break, time=0, value=1}, <at 1 0.05>, <at 1 0 RCIM_Cubic>]}}})
     B=&FNTimeOfDayScalar(Night Scale=0.5, Morning Scale=0.6, Evening Scale=0.5)
    )
    Scale RGB=&MakeVectorFromLinearColorRGB(Color=&FNTimeOfDayColor(
      Night Color=<rgba 0.4463 0.62 1 1>
      Morning Color=<rgba 1 0.6899 0.5573 1>
      Day Color=<rgba 1 0.7878 0.6667 1>
      Evening Color=<rgba 1 0.6291 0.4464 1>
     ))
    ScaleA=true
    ScaleRGB=true
   ) {module=ScaleColor002}
   /Game/Effects/Niagara/Modules/AnimTrail/FN_UpdateAnimTrail/FN_UpdateAnimTrail_SocketReader/FN_UpdateAnimTrail_SocketReader(Socket Reader:NiagaraDataInterfaceSocketReader=$System.Socket Reader)
  ]
  rendererClasses [&NiagaraRibbonRendererProperties]
  renderers [{rendererClass=&NiagaraRibbonRendererProperties}]
 }
 Trail {
  emitterUpdateScript [
   &EmitterState(Life Cycle Mode=::NewEnumerator1, Loop Behavior=::NewEnumerator1, Loop Duration=0.5, UseLoopDelay=false)
   &FN_SpawnAnimTrail_v2()
  ]
  particleSpawnScript [
   &FN_InitializeAnimTrailBoneIndecies(Second Bone or Socket Index=1)
   &FN_InitializeAnimTrail_SocketReader(
    AnimTrailWidth=0.95
    Color:LinearColor=&MultiplyLinearColor(
     Linear Color=&MergeHelper_ColorSwitch(
      0=<rgba 0.35 0.385 0.5 0.4>
      1=<rgba 0.0998 0.4368 0.5573 0.4>
      2=?
      3=?
      Index=&MakeIntFromFloat(Float=$User.VariantIndex)
      Num=::NewEnumerator0
     )
     Scale Factor=&MakeFloat4FromVectorAndFloat(
      "Vector (RGB)"=<1 1 1>
      "Float (Alpha)"=&FloatFromCurve(FloatCurve=DataInterface() {propertyValues {Curve {keys [<at 0.2 1>,<at 0.25 1>]}}}, CurveIndex=$System.Age)
     )
    )
    Lifetime=0.1
    Socket Reader:NiagaraDataInterfaceSocketReader=$System.Socket Reader
   )
   &AddVelocity(Velocity=<2000 0 0>) {enabled=false}
  ]
  particleUpdateScript [
   &SolveForcesAndVelocity(Clamp Velocity=false, Limit Acceleration=false)
   &ScaleColor(
    Scale Alpha=&Multiply_Float(
     A=&FloatFromCurve(FloatCurve=DataInterface() {propertyValues {Curve {keys [{interpMode=RCIM_Cubic, leaveTangent=-2.3646, tangentMode=RCTM_Break, time=0, value=1}, <at 0.5 0.05>, <at 1 0 RCIM_Cubic>]}}})
     B=&FNTimeOfDayScalar(Night Scale=0.5, Morning Scale=0.6, Evening Scale=0.5)
    )
    Scale RGB=&MakeVectorFromLinearColorRGB(Color=&FNTimeOfDayColor(
      Night Color=<rgba 0.4463 0.62 1 1>
      Morning Color=<rgba 1 0.6899 0.5573 1>
      Day Color=<rgba 1 0.7878 0.6667 1>
      Evening Color=<rgba 1 0.6291 0.4464 1>
     ))
    ScaleA=true
    ScaleRGB=true
   ) {module=ScaleColor002}
   &FN_UpdateAnimTrail_SocketReader(Socket Reader:NiagaraDataInterfaceSocketReader=$System.Socket Reader)
  ]
  rendererClasses [&NiagaraRibbonRendererProperties]
  renderers [{rendererClass=&NiagaraRibbonRendererProperties}]
 }
}
summary {
 systemName=NS_Pickaxe_CasinoReaper_Swing
 emitters [
  {emitterName=Trail, rendererClasses [&NiagaraRibbonRendererProperties]}
  {emitterName=TipTrail, rendererClasses [&NiagaraRibbonRendererProperties]}
 ]
 userVariables [
  User.RibbonIDOffset {type=NiagaraInt32}
  User.VariantIndex {type=NiagaraFloat, defaultValue=0, description=""}
 ]
}
